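# Physics asset for the hood of the medium male player's Goat Robe body, covering the head and five dynamic hood bones.
# Each bone gets a capsule body: a large one on the head and small ones on the upper and lower left and right hood bones and the tip.
# Every hood body is constrained directly to the head rather than chained, with tight swing and twist limits that are looser on the upper sides than the lower.
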PhysicsAsset /Game/Characters/Player/Male/Medium/Bodies/M_MED_GoatRobe/Meshes/Parts/M_MED_Goat_Robe_Physics
bodies {
 dyn_hood_lower_l [{shapeType=Capsule, length=2.5, radius=2, shapeName=dyn_hood_lower_l_capsule, center=<2.1295 0 0>, rotation=<rot 90 0 0>}]
 dyn_hood_lower_r [{shapeType=Capsule, length=3.25, radius=2, shapeName=dyn_hood_lower_r_capsule, center=<1.535 0 0>, rotation=<rot 90 0 0>}]
 dyn_hood_tip [{shapeType=Capsule, length=0.505, radius=1.2174, center=<0 0 -1.1454>}]
 dyn_hood_upper_l [{shapeType=Capsule, length=2.5, radius=2, shapeName=dyn_hood_upper_l_capsule, center=<2.1295 0 0>, rotation=<rot 90 0 0>}]
 dyn_hood_upper_r [{shapeType=Capsule, length=3.25, radius=2, shapeName=dyn_hood_upper_r_capsule, center=<1.535 0 0>, rotation=<rot 90 0 0>}]
 head [{shapeType=Capsule, length=8.6093, radius=7.0271, rotation=<rot 90 0 0>}]
}
constraints [
 {bone1Name=dyn_hood_lower_r, bone2Name=head, swing1LimitDegrees=12, swing2LimitDegrees=4, twistLimitDegrees=2}
 {bone1Name=dyn_hood_upper_r, bone2Name=head, swing1LimitDegrees=20, swing2LimitDegrees=10, twistLimitDegrees=2}
 {bone1Name=dyn_hood_upper_l, bone2Name=head, swing1LimitDegrees=20, swing2LimitDegrees=10, twistLimitDegrees=2}
 {bone1Name=dyn_hood_lower_l, bone2Name=head, swing1LimitDegrees=12, swing2LimitDegrees=4, twistLimitDegrees=2}
 {bone1Name=dyn_hood_tip, bone2Name=head, swing1LimitDegrees=10, swing2LimitDegrees=10, twistLimitDegrees=6}
]
